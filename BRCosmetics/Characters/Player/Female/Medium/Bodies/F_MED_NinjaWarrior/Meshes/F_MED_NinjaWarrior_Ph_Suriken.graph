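# Physics asset for the female medium NinjaWarrior body, covering a hanging shuriken prop and the front-left skirt, apparently for secondary motion.
# The skirt is a three-link capsule chain constrained back to a tiny Base capsule, with twist left free on each joint.
# The shuriken is a flat box constrained to its own tiny Base box, with twist locked.
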PhysicsAsset /BRCosmetics/Characters/Player/Female/Medium/Bodies/F_MED_NinjaWarrior/Meshes/F_MED_NinjaWarrior_Ph_Suriken
bodies {
 Base_skirt_Fr_L [{shapeType=Capsule, length=0.505, radius=0.5, shapeName=Base_skirt_Fr_L_capsule, rotation=<rot 0 -90 0>}]
 Base_suriken [{shapeType=Box, extentX=1.01, extentY=1.01, extentZ=1.01, shapeName=Base_suriken_box, rotation=<rot 0 -90 0>}]
 Dyn_skirt_Fr_01_L [{
   shapeType=Capsule
   length=6.5
   radius=4
   shapeName=Dyn_skirt_Fr_01_L_capsule
   center=<3.9926 3.9868 -2.5819>
   rotation=<rot 0.0005 -4.9999 -70>
  }]
 Dyn_skirt_Fr_02_L [{
   shapeType=Capsule
   length=9.5
   radius=4
   shapeName=Dyn_skirt_Fr_02_L_capsule
   center=<3.9926 3.9868 -2.5819>
   rotation=<rot 0.0005 -4.9999 -70.0001>
  }]
 Dyn_skirt_Fr_03_L [{
   shapeType=Capsule
   length=9.5
   radius=4
   shapeName=Dyn_skirt_Fr_03_L_capsule
   center=<3.4193 3.2371 -4.6497>
   rotation=<rot 0.0005 5.0002 -70>
  }]
 Dyn_suriken [{
   shapeType=Box
   extentX=7.4925
   extentY=1.7961
   extentZ=7.4887
   shapeName=Dyn_suriken_box
   center=<8.4446 -0.6 -1.15>
   rotation=<rot 5.3978 -94.7568 -83.5341>
  }]
}
constraints [
 {bone1Name=Dyn_skirt_Fr_01_L, bone2Name=Base_skirt_Fr_L, swing1LimitDegrees=45, swing2LimitDegrees=45, twistLimitDegrees=45, twistMotion=Free}
 {
  bone1Name=Dyn_skirt_Fr_02_L
  bone2Name=Dyn_skirt_Fr_01_L
  swing1LimitDegrees=45
  swing2LimitDegrees=45
  twistLimitDegrees=45
  twistMotion=Free
 }
 {
  bone1Name=Dyn_skirt_Fr_03_L
  bone2Name=Dyn_skirt_Fr_02_L
  swing1LimitDegrees=45
  swing2LimitDegrees=45
  twistLimitDegrees=45
  twistMotion=Free
 }
 {bone1Name=Dyn_suriken, bone2Name=Base_suriken, swing1LimitDegrees=30, swing2LimitDegrees=45, twistLimitDegrees=11, twistMotion=Locked}
]
