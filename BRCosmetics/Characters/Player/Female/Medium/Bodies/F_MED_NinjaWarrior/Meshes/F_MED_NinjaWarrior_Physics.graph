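# Physics asset for the female medium NinjaWarrior player body cosmetic, covering only its kimono and skirt bone chains and the pelvis.
# Capsule bodies form chains from the pelvis through Base bones into Dyn segments: four kimono panels, plus back, side and front-left skirt panels.
# Kimono segments use limited or locked angular joints, while skirt joints rotate freely.
# Extra cross constraints link the two back skirt tips and tie each back kimono chain to its same-side neighbor.
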PhysicsAsset /BRCosmetics/Characters/Player/Female/Medium/Bodies/F_MED_NinjaWarrior/Meshes/F_MED_NinjaWarrior_Physics
bodies {
 Base_kimono_bk_l [{shapeType=Capsule, length=0.505, radius=0.505, shapeName=Base_kimono_bk_l_capsule, rotation=<rot 0 -90 0>}]
 Base_kimono_bk_r [{shapeType=Capsule, length=0.505, radius=0.505, shapeName=Base_kimono_bk_r_capsule, rotation=<rot 0 -90 0>}]
 Base_kimono_l [{shapeType=Capsule, length=0.505, radius=0.505, shapeName=Base_kimono_l_capsule, rotation=<rot 0 -90 0>}]
 Base_kimono_r [{shapeType=Capsule, length=0.505, radius=0.505, shapeName=Base_kimono_r_capsule, rotation=<rot 0 -90 0>}]
 Base_skirt_Bk_L [{shapeType=Capsule, length=0, radius=2.005, shapeName=Base_skirt_Bk_L_capsule, center=<-0.1772 0.7336 0>, rotation=<rot 0 -90 0>}]
 Base_skirt_Bk_R [{shapeType=Capsule, length=0, radius=2.005, shapeName=Base_skirt_Bk_R_capsule, center=<-0.7336 -0.1772 0>}]
 Base_skirt_Fr_L [{shapeType=Capsule, length=0.505, radius=0.505, shapeName=Base_skirt_Fr_L_capsule, rotation=<rot 0 -90 0>}]
 Base_skirt_Side_L [{shapeType=Capsule, length=0.505, radius=0.505, shapeName=Base_skirt_Side_L_capsule, rotation=<rot 0 -90 0>}]
 Base_skirt_Side_R [{shapeType=Capsule, length=0.505, radius=0.505, shapeName=Base_skirt_Side_R_capsule, center=<3.4817 0.5014 -1.8714>, rotation=<rot 0 0 75>}]
 Dyn_kimono_01 [{shapeType=Capsule, length=14.5, radius=5.5, shapeName=Dyn_kimono_01_capsule, center=<7.7264 3.9371 0.3444>, rotation=<rot 0 0 -5>}]
 Dyn_kimono_02 [{shapeType=Capsule, length=13, radius=6.5, shapeName=Dyn_kimono_02_capsule, center=<7.2215 4.7356 0.4143>, rotation=<rot -0.0001 0 -5>}]
 Dyn_kimono_bk_l_01 [{
   shapeType=Capsule
   length=10
   radius=7
   shapeName=Dyn_kimono_bk_l_01_capsule
   center=<3.1521 -0.9071 5.1446>
   rotation=<rot -1.7284 -9.851 80.149>
  }]
 Dyn_kimono_bk_l_02 [{shapeType=Capsule, length=10, radius=7, shapeName=Dyn_kimono_bk_l_02_capsule, center=<7 -1 3>, rotation=<rot -0.0001 0 80>}]
 Dyn_kimono_bk_r_01 [{
   shapeType=Capsule
   length=13
   radius=7
   shapeName=Dyn_kimono_bk_r_01_capsule
   center=<3.1438 -0.5353 3.5787>
   rotation=<rot 0.8666 4.9243 85.0373>
  }]
 Dyn_kimono_bk_r_02 [{shapeType=Capsule, length=13, radius=7, shapeName=Dyn_kimono_bk_r_02_capsule, center=<7 -1 3>, rotation=<rot -0.0001 0 80>}]
 Dyn_kimono_r_01 [{shapeType=Capsule, length=9, radius=4.9, shapeName=Dyn_kimono_r_01_capsule, center=<5.6654 5.6 0>}]
 Dyn_kimono_r_02 [{shapeType=Capsule, length=8, radius=5, shapeName=Dyn_kimono_r_02_capsule, center=<5.965 5.2 -1.6216>}]
 Dyn_skirt_Bk_01_L [{shapeType=Capsule, length=7, radius=7, shapeName=Dyn_skirt_Bk_01_L_capsule, center=<0.8531 0.8756 -2.4056>, rotation=<rot -0.0004 0.0001 70>}]
 Dyn_skirt_Bk_01_R [{
   shapeType=Capsule
   length=7
   radius=7
   shapeName=Dyn_skirt_Bk_01_R_capsule
   center=<0.8531 -0.6626 -2.4727>
   rotation=<rot -0.0007 -0.0002 105>
  }]
 Dyn_skirt_Bk_02_L [{shapeType=Capsule, length=7, radius=7, shapeName=Dyn_skirt_Bk_02_L_capsule, center=<0.8531 0.8756 -2.4056>, rotation=<rot -0.0004 0.0001 70>}]
 Dyn_skirt_Bk_02_R [{
   shapeType=Capsule
   length=7
   radius=7
   shapeName=Dyn_skirt_Bk_02_R_capsule
   center=<0.8531 -0.6626 -2.4727>
   rotation=<rot -0.0007 -0.0002 105>
  }]
 Dyn_skirt_Fr_01_L [{shapeType=Capsule, length=6.5, radius=5, shapeName=Dyn_skirt_Fr_01_L_capsule, center=<4.7837 0 0>, rotation=<rot -0.8669 -4.9243 -99.9627>}]
 Dyn_skirt_Fr_02_L [
  {shapeType=Capsule, length=6.5, radius=5, shapeName=Dyn_skirt_Fr_02_L_capsule, center=<4.7837 0 0>, rotation=<rot -0.8669 -4.9243 -99.9627>}
  {shapeType=Capsule, length=3.5, radius=5, shapeName=Dyn_skirt_Fr_02_L_capsule, center=<5.0604 10.1551 -0.4337>, rotation=<rot 0.0001 0 90>}
 ]
 Dyn_skirt_Fr_03_L [{shapeType=Capsule, length=6.5, radius=5, shapeName=Dyn_skirt_Fr_03_L_capsule, center=<4.7837 0 0>, rotation=<rot 0.8674 4.9245 -99.9627>}]
 Dyn_skirt_Side_01_L [{shapeType=Capsule, length=5, radius=6, shapeName=Dyn_skirt_Side_01_L_capsule, center=<3.7803 -1.723 -2.138>, rotation=<rot 0 0.0001 95>}]
 Dyn_skirt_Side_01_R [{
   shapeType=Capsule
   length=6.0546
   radius=4.3283
   shapeName=Dyn_skirt_Side_01_R_capsule
   center=<3.4817 0.5014 -1.8714>
   rotation=<rot 0 0 75>
  }]
 Dyn_skirt_Side_02_L [{shapeType=Capsule, length=5, radius=6, shapeName=Dyn_skirt_Side_02_L_capsule, center=<3.7803 -1.723 -2.138>, rotation=<rot 0 0.0001 95>}]
 Dyn_skirt_Side_02_R [{
   shapeType=Capsule
   length=2.2086
   radius=4.1221
   shapeName=Dyn_skirt_Side_02_R_capsule
   center=<3.4817 -1.5613 -1.9812>
   rotation=<rot 0 0 75>
  }]
 Dyn_skirt_Side_03_L [{shapeType=Capsule, length=5, radius=4, shapeName=Dyn_skirt_Side_03_L_capsule, center=<3.7803 -1.723 -2.138>, rotation=<rot 0 0.0001 95>}]
 Dyn_skirt_Side_03_R [{
   shapeType=Capsule
   length=7.5033
   radius=3.8754
   shapeName=Dyn_skirt_Side_03_R_capsule
   center=<3.4817 0.5014 -1.8714>
   rotation=<rot 0 0 75>
  }]
 pelvis [{shapeType=Capsule, length=13.2942, radius=9.5013, shapeName=pelvis_capsule, center=<0.4346 -2.6395 0.1569>, rotation=<rot 3.9141 -90 1.6864>}]
}
constraints [
 {bone1Name=Dyn_kimono_bk_l_01, bone2Name=Base_kimono_bk_l, swing1LimitDegrees=15, swing2LimitDegrees=15, twistLimitDegrees=5}
 {bone1Name=Dyn_kimono_bk_l_02, bone2Name=Dyn_kimono_bk_l_01, swing1LimitDegrees=45, swing2LimitDegrees=45, twistLimitDegrees=5}
 {bone1Name=Dyn_kimono_bk_r_01, bone2Name=Base_kimono_bk_r, swing1LimitDegrees=15, swing2LimitDegrees=15, twistLimitDegrees=5}
 {bone1Name=Dyn_kimono_bk_r_02, bone2Name=Dyn_kimono_bk_r_01, swing1LimitDegrees=45, swing2LimitDegrees=45, twistLimitDegrees=5}
 {
  bone1Name=Base_kimono_bk_l
  bone2Name=pelvis
  swing1LimitDegrees=45
  swing1Motion=Locked
  swing2LimitDegrees=45
  swing2Motion=Locked
  twistLimitDegrees=45
  twistMotion=Locked
 }
 {
  bone1Name=Base_kimono_bk_r
  bone2Name=pelvis
  swing1LimitDegrees=45
  swing1Motion=Locked
  swing2LimitDegrees=45
  swing2Motion=Locked
  twistLimitDegrees=45
  twistMotion=Locked
 }
 {
  bone1Name=Base_skirt_Bk_L
  bone2Name=pelvis
  swing1LimitDegrees=45
  swing1Motion=Free
  swing2LimitDegrees=45
  swing2Motion=Free
  twistLimitDegrees=45
  twistMotion=Free
 }
 {
  bone1Name=Dyn_skirt_Bk_01_L
  bone2Name=Base_skirt_Bk_L
  swing1LimitDegrees=45
  swing1Motion=Free
  swing2LimitDegrees=45
  swing2Motion=Free
  twistLimitDegrees=45
  twistMotion=Free
 }
 {
  bone1Name=Dyn_skirt_Bk_02_L
  bone2Name=Dyn_skirt_Bk_01_L
  swing1LimitDegrees=45
  swing1Motion=Free
  swing2LimitDegrees=45
  swing2Motion=Free
  twistLimitDegrees=45
  twistMotion=Free
 }
 {
  bone1Name=Base_skirt_Bk_R
  bone2Name=pelvis
  swing1LimitDegrees=45
  swing1Motion=Free
  swing2LimitDegrees=45
  swing2Motion=Free
  twistLimitDegrees=45
  twistMotion=Free
 }
 {
  bone1Name=Dyn_skirt_Bk_01_R
  bone2Name=Base_skirt_Bk_R
  swing1LimitDegrees=45
  swing1Motion=Free
  swing2LimitDegrees=45
  swing2Motion=Free
  twistLimitDegrees=45
  twistMotion=Free
 }
 {
  bone1Name=Dyn_skirt_Bk_02_R
  bone2Name=Dyn_skirt_Bk_01_R
  swing1LimitDegrees=45
  swing1Motion=Free
  swing2LimitDegrees=45
  swing2Motion=Free
  twistLimitDegrees=45
  twistMotion=Free
 }
 {
  bone1Name=Base_skirt_Fr_L
  bone2Name=pelvis
  swing1LimitDegrees=45
  swing1Motion=Free
  swing2LimitDegrees=45
  swing2Motion=Free
  twistLimitDegrees=45
  twistMotion=Free
 }
 {
  bone1Name=Dyn_skirt_Fr_01_L
  bone2Name=Base_skirt_Fr_L
  swing1LimitDegrees=45
  swing1Motion=Free
  swing2LimitDegrees=45
  swing2Motion=Free
  twistLimitDegrees=45
  twistMotion=Free
 }
 {
  bone1Name=Dyn_skirt_Fr_02_L
  bone2Name=Dyn_skirt_Fr_01_L
  swing1LimitDegrees=45
  swing1Motion=Free
  swing2LimitDegrees=45
  swing2Motion=Free
  twistLimitDegrees=45
  twistMotion=Free
 }
 {
  bone1Name=Dyn_skirt_Fr_03_L
  bone2Name=Dyn_skirt_Fr_02_L
  swing1LimitDegrees=45
  swing1Motion=Free
  swing2LimitDegrees=45
  swing2Motion=Free
  twistLimitDegrees=45
  twistMotion=Free
 }
 {
  bone1Name=Base_skirt_Side_L
  bone2Name=pelvis
  swing1LimitDegrees=45
  swing1Motion=Free
  swing2LimitDegrees=45
  swing2Motion=Free
  twistLimitDegrees=45
  twistMotion=Free
 }
 {
  bone1Name=Dyn_skirt_Side_01_L
  bone2Name=Base_skirt_Side_L
  swing1LimitDegrees=45
  swing1Motion=Free
  swing2LimitDegrees=45
  swing2Motion=Free
  twistLimitDegrees=45
  twistMotion=Free
 }
 {
  bone1Name=Dyn_skirt_Side_02_L
  bone2Name=Dyn_skirt_Side_01_L
  swing1LimitDegrees=45
  swing1Motion=Free
  swing2LimitDegrees=45
  swing2Motion=Free
  twistLimitDegrees=45
  twistMotion=Free
 }
 {
  bone1Name=Dyn_skirt_Side_03_L
  bone2Name=Dyn_skirt_Side_02_L
  swing1LimitDegrees=45
  swing1Motion=Free
  swing2LimitDegrees=45
  swing2Motion=Free
  twistLimitDegrees=45
  twistMotion=Free
 }
 {
  bone1Name=Base_kimono_l
  bone2Name=pelvis
  swing1LimitDegrees=45
  swing1Motion=Free
  swing2LimitDegrees=45
  swing2Motion=Free
  twistLimitDegrees=45
  twistMotion=Free
 }
 {bone1Name=Dyn_kimono_01, bone2Name=Base_kimono_l, swing1LimitDegrees=35, swing2LimitDegrees=15, twistLimitDegrees=5}
 {bone1Name=Dyn_kimono_02, bone2Name=Dyn_kimono_01, swing1LimitDegrees=35, swing2LimitDegrees=15, twistLimitDegrees=5}
 {
  bone1Name=Dyn_skirt_Bk_02_R
  bone2Name=Dyn_skirt_Bk_02_L
  swing1LimitDegrees=45
  swing1Motion=Free
  swing2LimitDegrees=45
  swing2Motion=Free
  twistLimitDegrees=45
  twistMotion=Free
 }
 {
  bone1Name=Dyn_kimono_bk_l_01
  bone2Name=Dyn_kimono_01
  swing1LimitDegrees=45
  swing1Motion=Locked
  swing2LimitDegrees=45
  swing2Motion=Locked
  twistLimitDegrees=45
  twistMotion=Locked
 }
 {
  bone1Name=Base_skirt_Side_R
  bone2Name=pelvis
  swing1LimitDegrees=45
  swing1Motion=Free
  swing2LimitDegrees=45
  swing2Motion=Free
  twistLimitDegrees=45
  twistMotion=Free
 }
 {
  bone1Name=Dyn_skirt_Side_01_R
  bone2Name=Base_skirt_Side_R
  swing1LimitDegrees=45
  swing1Motion=Free
  swing2LimitDegrees=45
  swing2Motion=Free
  twistLimitDegrees=45
  twistMotion=Free
 }
 {
  bone1Name=Dyn_skirt_Side_02_R
  bone2Name=Dyn_skirt_Side_01_R
  swing1LimitDegrees=45
  swing1Motion=Free
  swing2LimitDegrees=45
  swing2Motion=Free
  twistLimitDegrees=45
  twistMotion=Free
 }
 {
  bone1Name=Dyn_skirt_Side_03_R
  bone2Name=Dyn_skirt_Side_02_R
  swing1LimitDegrees=45
  swing1Motion=Free
  swing2LimitDegrees=45
  swing2Motion=Free
  twistLimitDegrees=45
  twistMotion=Free
 }
 {
  bone1Name=Base_kimono_r
  bone2Name=pelvis
  swing1LimitDegrees=45
  swing1Motion=Free
  swing2LimitDegrees=45
  swing2Motion=Free
  twistLimitDegrees=45
  twistMotion=Free
 }
 {bone1Name=Dyn_kimono_r_01, bone2Name=Base_kimono_r, swing1LimitDegrees=30, swing2LimitDegrees=15, swing2Motion=Locked, twistLimitDegrees=5}
 {bone1Name=Dyn_kimono_r_02, bone2Name=Dyn_kimono_r_01, swing1LimitDegrees=30, swing2LimitDegrees=15, swing2Motion=Locked, twistLimitDegrees=5}
 {
  bone1Name=Dyn_kimono_bk_r_02
  bone2Name=Dyn_kimono_r_02
  swing1LimitDegrees=45
  swing1Motion=Free
  swing2LimitDegrees=45
  swing2Motion=Free
  twistLimitDegrees=45
  twistMotion=Free
 }
]
